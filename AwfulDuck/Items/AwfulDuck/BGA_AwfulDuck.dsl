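; Abduction beam gameplay actor for AwfulDuck that lifts one pawn or a list of abductees after a warm-up.
; An intro sound leads into the warm-up, where a timeline shortens the interval of warning pulse events and a nearby glow controller brightens.
; It then applies the abduct GameplayEffect, toggles audio and post-process until ramp-down, and self-destroys in single-pawn mode.
; OnRep_Abductees diffs against the previous list to add or remove warning effects.
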(event Custom|StartPulse
  ;[1 1 1] Player Pulse Fx Warning
  ;;[1 1 1] Ramp up Pulse Speed
  (Utilities|FlowControl|Sequence
    (:then_0
      (Components|Timeline|SetPlayRate (Variables|BGA_AwfulDuck|GetPulseRamp) (/ 1.0 (Variables|Default|GetWarmUpTime)))
      (bind (_direction _newtrack_0) (Timeline|PulseRamp PlayFromStart)
        (:Update
          (Variables|Default|SetPulseTime (Math|Float|Lerp 5.0 0.2 _newtrack_0)))))
    (:then_1
      (CallFunction|PulseTimer))))

(event Custom|PlayerPulse
  (bind _pulsetime (Variables|Default|GetPulseTime))
  (bind _gameplayeventdata (Utilities|Struct|MakeGameplayEventData "(TagName=\"\")" 0 0 0 0 "()" "(GameplayTags=)" "(GameplayTags=)" _pulsetime))
  ;[1 1 1] Player Pulse Fx Warning
  (Utilities|FlowControl|Gate false
    (:Exit
      (if (Variables|Default|GetSingleAbduction)
        (Ability|SendGameplayEventtoActor (Variables|Default|GetSinglePawn) "(TagName=\"Event.Athena.Item.AwfulDuck.WarmUp.Pulse\")" _gameplayeventdata)
        (else
          (for _array_element (Variables|Default|GetAbductees)
            (Ability|SendGameplayEventtoActor _array_element "(TagName=\"Event.Athena.Item.AwfulDuck.WarmUp.Pulse\")" _gameplayeventdata))))
      (CallFunction|PulseTimer))))

(event Custom|ClosePulseGate
  (bind _pulsetime (Variables|Default|GetPulseTime))
  (bind _gameplayeventdata (Utilities|Struct|MakeGameplayEventData "(TagName=\"\")" 0 0 0 0 "()" "(GameplayTags=)" "(GameplayTags=)" _pulsetime))
  ;[1 1 1] Player Pulse Fx Warning
  (Utilities|FlowControl|Gate false
    (:Exit
      (if (Variables|Default|GetSingleAbduction)
        (Ability|SendGameplayEventtoActor (Variables|Default|GetSinglePawn) "(TagName=\"Event.Athena.Item.AwfulDuck.WarmUp.Pulse\")" _gameplayeventdata)
        (else
          (for _array_element (Variables|Default|GetAbductees)
            (Ability|SendGameplayEventtoActor _array_element "(TagName=\"Event.Athena.Item.AwfulDuck.WarmUp.Pulse\")" _gameplayeventdata))))
      (CallFunction|PulseTimer))))

(event Custom|StartSingleAbduction
  (params (Pawn :type Fort%20Pawn%20Object%20Reference))
  (bind _ns_awfulduck_lightbeam_core (Variables|Default|GetNS_AwfulDuck_LightBeam_Core))
  (bind _output_get (Variables|Default|SetSinglePawn Pawn))
  (Utilities|IsValid _output_get
    (:Is%20Valid
      (Variables|Default|SetSingleAbduction true)
      ;[1 1 1] Central light
      (Components|Activation|Activate _ns_awfulduck_lightbeam_core true) ; Turn on central effects
      (Niagara|SetNiagaraVariable(Float) _ns_awfulduck_lightbeam_core "BurstEffectSpawnProbability" 1.0)
      (Niagara|SetNiagaraVariable(Float) _ns_awfulduck_lightbeam_core "FlareFlicker" 1.0)
      (bind _returnvalue (Audio|SpawnSoundAttached "/AwfulDuck/Audio/AwfulDuck_Intro_Cue.AwfulDuck_Intro_Cue" (Variables|Default|GetNS_AwfulDuck_LightBeam_Core))) ; Warm up sound
      (Variables|Default|SetLightCoreAudioComponent _returnvalue)
      (Utilities|Time|SetTimerbyEvent (delegate Custom|Warm%20up%20sound%20complete) (Variables|Default|GetAudioWarmUpTime)))))

(event Custom|LiftUpPlayers
  ;[1 1 1] Lift up player(s)
  (if (Variables|Default|GetSingleAbduction)
    (CallFunction|AbductPawn (Variables|Default|GetSinglePawn))
    (else
      (for _array_element (Variables|Default|GetAbductees)
        (CallFunction|AbductPawn _array_element)))))

(event Custom|PulseTimer
  ;[1 1 1] Player Pulse Fx Warning
  (Utilities|Time|SetTimerbyFunctionName 0 "PlayerPulse" (Variables|Default|GetPulseTime)))

(event Custom|StartAbduction
  (bind _ns_awfulduck_lightbeam_core (Variables|Default|GetNS_AwfulDuck_LightBeam_Core))
  (CallFunction|GetGlowController) ; Pre popluate controller
  ;[1 1 1] Central light
  (Components|Activation|Activate _ns_awfulduck_lightbeam_core true) ; Turn on central effects
  (Niagara|SetNiagaraVariable(Float) _ns_awfulduck_lightbeam_core "BurstEffectSpawnProbability" 1.0)
  (Niagara|SetNiagaraVariable(Float) _ns_awfulduck_lightbeam_core "FlareFlicker" 1.0)
  (bind _returnvalue (Audio|SpawnSoundAttached "/AwfulDuck/Audio/AwfulDuck_Intro_Cue.AwfulDuck_Intro_Cue" (Variables|Default|GetNS_AwfulDuck_LightBeam_Core))) ; Warm up sound
  (Variables|Default|SetLightCoreAudioComponent _returnvalue)
  (Utilities|Time|SetTimerbyEvent (delegate Custom|Warm%20up%20sound%20complete) (Variables|Default|GetAudioWarmUpTime)))

(event Custom|Warm%20up%20sound%20complete
  (bind _pulsetime (Variables|Default|GetPulseTime))
  (bind _gameplayeventdata (Utilities|Struct|MakeGameplayEventData "(TagName=\"\")" 0 0 0 0 "()" "(GameplayTags=)" "(GameplayTags=)" _pulsetime))
  (bind _lightcoreaudiocomponent (Variables|Default|GetLightCoreAudioComponent))
  (if (Variables|Default|GetSingleAbduction)
    (CallFunction|AddWarningEffect (Variables|Default|GetSinglePawn)))
  (CallFunction|InitfromRows)
  (Utilities|FlowControl|Sequence
    (:then_0
      ;[1 1 1] Warm Up Abduction Portal
      (CallFunction|StartPulse)
      (bind (_result _success) (CallFunction|GetGlowController))
      (if _success
        (Class|BPLuxurySupremeGlowController|BeginWarmUp _result (Variables|Default|GetWarmUpTime)))
      (Utilities|FlowControl|Delay (Variables|Default|GetWarmUpTime))
      ;[1 1 1] Start abduction
      (Development|PrintString "Lift players")
      (CallFunction|LiftUpPlayers)
      (CallFunction|SetAbductVFXActive true)
      (CallFunction|ClosePulseGate)
      ;[1 1 1] Destroy
      (Utilities|FlowControl|Delay (Variables|Default|GetRampDownTime))
      (CallFunction|SetAbductVFXActive false)
      (bind (_result_2 _success_2) (CallFunction|GetGlowController))
      (if _success_2
        (Class|BPLuxurySupremeGlowController|BeginCoolDown _result_2))
      (Audio|Components|Audio|Stop _lightcoreaudiocomponent)
      (Components|DestroyComponent _lightcoreaudiocomponent)
      (if (Variables|Default|GetSingleAbduction)
        (Actor|DestroyActor)))
    (:then_1
      ;[1 1 1] Player Pulse Fx Warning
      (Utilities|FlowControl|Gate false
        (:Exit
          (if (Variables|Default|GetSingleAbduction)
            (Ability|SendGameplayEventtoActor (Variables|Default|GetSinglePawn) "(TagName=\"Event.Athena.Item.AwfulDuck.WarmUp.Pulse\")" _gameplayeventdata)
            (else
              (for _array_element (Variables|Default|GetAbductees)
                (Ability|SendGameplayEventtoActor _array_element "(TagName=\"Event.Athena.Item.AwfulDuck.WarmUp.Pulse\")" _gameplayeventdata))))
          (CallFunction|PulseTimer))))))

(event Custom|SetAbductVFXActive
  (params (Active :type Boolean))
  (bind _ambientaudio_close (Variables|Default|GetAmbientAudio_Close))
  ;[1 1 1] VFX Switches
  (if Active
    (Components|Activation|Activate _ambientaudio_close true)
    (Variables|/Script/Engine.PostProcessComponent|SetbEnabled true (Variables|Default|GetPostProcess))
    (else
      (Components|Activation|Deactivate _ambientaudio_close)
      (Variables|/Script/Engine.PostProcessComponent|SetbEnabled false (Variables|Default|GetPostProcess)))))

(fn AbductPawn
  (params (Pawn :type Actor%20Object%20Reference))
  (Fortnite|ApplyGameplayEffectsToActor Pawn (Utilities|Array|MakeArray "/AwfulDuck/Items/AwfulDuck/GE_Athena_AwfulDuck_Abduct.GE_Athena_AwfulDuck_Abduct_C") (Transformation|GetActorLocation) false 0 (Ability|GetFortAbilitySystemComponent self)))

(fn AddWarningEffect
  (params (Pawn :type Actor%20Object%20Reference))
  (Fortnite|ApplyGameplayEffectsToActor Pawn (Utilities|Array|MakeArray "/AwfulDuck/Items/AwfulDuck/GE_Athena_AwfulDuck_WarmUpPulse.GE_Athena_AwfulDuck_WarmUpPulse_C") (Transformation|GetActorLocation) false 0 (Ability|GetFortAbilitySystemComponent self))
  (Development|PrintString (Utilities|String|Append "Added Abduction Warning Effect to player " (Fortnite|GetPlayerNameSafe Pawn))))

(fn FilterAbductees
  (params (Actors :type Array%20of%20Actor%20Object%20References))
  (for _array_element Actors
    (bind _asfort_player_pawn_athena (Utilities|Casting|CastToFortPlayerPawnAthena _array_element)
      (:then
        (bind _bdidactorhavetags (Class|AthenaQuestFunctionLibrary|AthenaCheckActorHasTags _asfort_player_pawn_athena "(GameplayTags=((TagName=\"Pawn.Athena.NPC\")))" true))
        (bind _bdidactorhavetags_2 (Class|AthenaQuestFunctionLibrary|AthenaCheckActorHasTags _asfort_player_pawn_athena "(GameplayTags=((TagName=\"Pawn.Athena.NonParticipant\")))" true))
        (if (not (or _bdidactorhavetags_2 _bdidactorhavetags))
          (Utilities|Array|AddUnique (LocalVariables|GetTempActors) _asfort_player_pawn_athena)))
      (:CastFailed)))
  (return (LocalVariables|GetTempActors)))

(fn GetAbductionManager
  (if (not (Utilities|IsValid (Variables|Default|GetAbductionManager) :!class /Script/Engine.KismetSystemLibrary))
    (bind _returnvalue (Actor|GetActorOfClass "/AwfulDuck/Items/AwfulDuck/B_AwfulDuck_Manager.B_AwfulDuck_Manager_C"))
    (Variables|Default|SetAbductionManager _returnvalue))
  (return (Variables|Default|GetAbductionManager)))

(fn GetEligibleOverlappingPawns
  (Utilities|FlowControl|Sequence
    (:then_0
      (bind _outfortplayerpawns (Fortnite|GetFortPlayerPawnsInRange (Transformation|GetWorldLocation (Variables|Default|GetAbductCapsule)) (Components|Capsule|GetScaledCapsuleRadius (Variables|Default|GetAbductCapsule))))
      (LocalVariables|SetTemp%20Candidates _outfortplayerpawns)
      (bind _outfortplayerpawns_2 (Fortnite|GetFortPlayerPawnsInRange (Transformation|GetWorldLocation (Variables|Default|GetTopAbductCapsule)) (Components|Capsule|GetScaledCapsuleRadius (Variables|Default|GetTopAbductCapsule))))
      (Utilities|Array|AppendArray (LocalVariables|GetTemp%20Candidates) _outfortplayerpawns_2))
    (:then_1
      (bind _result (CallFunction|FilterAbductees (LocalVariables|GetTemp%20Candidates)))
      (return _result))))

(fn GetGlowController
  (bind _glowcontroller (Variables|Default|GetGlowController))
  (Utilities|IsValid _glowcontroller
    (:Is%20Valid
      (return _glowcontroller true))
    (:Is%20Not%20Valid
      (bind _outactors (Actor|GetAllActorsOfClass "/AwfulDuck/Athena/Apollo/Environments/FX/LuxurySupreme/BP_LuxurySupreme_GlowController.BP_LuxurySupreme_GlowController_C"))
      (for _array_element _outactors
        (if (< (Class|AbilityFunctions|GetDistanceBetweenActors _array_element self) 5e+03)
          (bind _output_get (Variables|Default|SetGlowController _array_element))
          (return _output_get true)))
      (return 0 false))))

(fn InitFromRows
  (Variables|Default|SetWarmUpTime (ScalableFloat|GetValueatLevel (Variables|Default|GetAwfulDuckWarmUpTime)))
  (Variables|Default|SetTargetSelectionBufferTime (ScalableFloat|GetValueatLevel (Variables|Default|GetAwfulDuckSelectBufferTime)))
  (Variables|Default|SetRampDownTime (ScalableFloat|GetValueatLevel (Variables|Default|GetAwfulDuckRampDownTime))))

(fn OnRep_Abductees
  (Utilities|FlowControl|Sequence
    (:then_0
      ;[0.1034 1 0.4975] For all Abductees, make sure they have the warning effect
      (for _array_element (Variables|Default|GetAbductees)
        (bind _asfort_player_pawn_athena (Utilities|Casting|CastToFortPlayerPawnAthena _array_element))
        (if (not (GameplayTags|HasMatchingGameplayTag _asfort_player_pawn_athena "(TagName=\"Athena.Gameplay.Status.MarkedForAbduct\")"))
          (CallFunction|AddWarningEffect _asfort_player_pawn_athena) ; In abductee list and not marked = add warning effect to mark
        )))
    (:then_1
      ;[0.892 0.5204 0.1874] For all PrevAbductees, if they are no longer in Abductees, make sure they have the effect removed
      (for _array_element_2 (Variables|Default|GetPrevAbductees)
        (bind _asfort_player_pawn_athena_2 (Utilities|Casting|CastToFortPlayerPawnAthena _array_element_2))
        (if (Utilities|Array|ContainsItem (Variables|Default|GetAbductees) _array_element_2)
          _
          (elif (GameplayTags|HasMatchingGameplayTag _asfort_player_pawn_athena_2 "(TagName=\"Athena.Gameplay.Status.MarkedForAbduct\")")
            (CallFunction|RemoveWarningEffect _asfort_player_pawn_athena_2) ; Not in the new list and marked = remove warning effect and mark
          ))))
    (:then_2
      ;[0.115 0.7197 1] Update PrevAbductees
      (Variables|Default|SetPrevAbductees (Variables|Default|GetAbductees)))))

(fn RemoveWarningEffect
  (params (Pawn :type Object%20Reference))
  (if (Networking|IsServer)
    (bind _asfort_player_pawn_athena (Utilities|Casting|CastToFortPlayerPawnAthena Pawn))
    (bind _returnvalue (Fortnite|Utilities|AttemptGetPlayerStateforPawn _asfort_player_pawn_athena))
    (GameplayEffects|RemoveActiveGameplayEffectbySourceEffect (Variables|/Script/FortniteGame.FortPlayerState|GetAbilitySystemComponent _returnvalue) "/AwfulDuck/Items/AwfulDuck/GE_Athena_AwfulDuck_WarmUpPulse.GE_Athena_AwfulDuck_WarmUpPulse_C" (Ability|GetFortAbilitySystemComponent self) -1)
    (Development|PrintString (Utilities|String|Append "Removed Abduction Warning Effect from player " (Fortnite|GetPlayerNameSafe _asfort_player_pawn_athena)))))

(fn ShouldDie
  (params (Damage :type Float%20%28single-precision%29) (EventInstigator :type Controller%20Object%20Reference) (DamageCauser :type Actor%20Object%20Reference) (bIsTest :type Boolean) (EffectContext :type Gameplay%20Effect%20Context%20Handle%20Structure%20%28by%20ref%29))
  (return false))

(fn ConstructionScript)
