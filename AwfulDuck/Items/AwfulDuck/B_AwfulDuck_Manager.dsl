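; Server-driven manager for AwfulDuck abductions, both random single-player grabs and mass abductions to squad-based Skyfire destinations.
; During storm holding, a calendar-scaled chance spawns BGA_AwfulDuck near a random eligible human.
; Mass abduction refreshes site candidates on a timer, culls whole squads to fit capacity, adds bots, then reports analytics.
; Squad IDs map abductees to tagged squad starts, gathered from Warmup onward; QA cheat events force tests.
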(event EventBeginPlay
  ;[0.5212 1 0.5368] Begin Play
  (|CallParentFunction)
  (if (Networking|IsServer)
    (Class|FortAsyncActionCurrentPlaylistReady|PlaylistDataReadyAsync
      (:OnReady
        (CallFunction|InitfromRows)
        (bind _returnvalue (GameplayEventRouter|GetEventRouter "/Script/FortniteGame.FortGameState" self))
        (bind (_asynctaskproxy _payload) (GameplayEventRouter|StartListeningToStatefulEvent _returnvalue "/Script/FortniteGame.FortGamePhaseUpdatedEvent")
          (:then
            (Variables|Default|SetGamePhaseUpdatedAsyncTask _asynctaskproxy)
            (bind (_asynctaskproxy_2 _payload_2) (GameplayEventRouter|StartListeningToStatefulEvent _returnvalue "/Script/FortniteGame.FortGamePhaseStepUpdatedEvent")
              (:then
                (Variables|Default|SetGamePhaseStepUpdatedAsyncTask _asynctaskproxy_2)
                (bind (_asynctaskproxy_3 _payload_3) (GameplayEventRouter|StartListeningToStatefulEvent _returnvalue "/Script/FortniteGame.FortCreativeStormShieldGamePhaseStepUpdatedEvent")
                  (:then
                    (Variables|Default|SetCreativeStormShieldGamePhaseUpdatedAsyncTask _asynctaskproxy_3))
                  (:OnEventReceived
                    _)
                  (:OnSavedEventStateExists
                    _))
                (CallFunction|PhaseStepChanged (.GamePhaseStep (Utilities|Struct|BreakFortCreativeStormShieldGamePhaseStepUpdatedEvent _payload_3))))
              (:OnEventReceived
                _)
              (:OnSavedEventStateExists
                _))
            (CallFunction|PhaseStepChanged (Utilities|Struct|BreakFortGamePhaseStepUpdatedEvent _payload_2)))
          (:OnEventReceived
            _)
          (:OnSavedEventStateExists
            _))
        (CallFunction|OnGamePhaseChanged (Utilities|Struct|BreakFortGamePhaseUpdatedEvent _payload))))))

(event Custom|OnGamePhaseChanged
  (params (GamePhase :type EAthenaGamePhase%20Enum))
  ;[0.191 0.8566 1] OnGamePhaseChanged - Wait until Warmup phase to initialize the skyfire destinations as their level overlay may not be loaded by the time PlaylistDataReady is called
  (if (Networking|IsServer)
    (if (Math|Byte|Byte>=Byte GamePhase (Utilities|Enum|LiteralenumEAthenaGamePhase "Warmup"))
      (Default|InitSkyfireDestinations)
      (bind _gamephaseupdatedasynctask (Variables|Default|GetGamePhaseUpdatedAsyncTask)
        (:then
          (GameplayEventRouter|StopListeningtoStatefulEvent _gamephaseupdatedasynctask))
        (:else)))))

(event Custom|TestAbductSinglePlayer
  (bind (_outitem _outindex) (Utilities|Array|RandomArrayItem (Variables|Default|GetSingleAbductionCandidates)))
  ;[1 1 1] Periodic abductions
  (Variables|Default|SetUsingSingleTestCheat true) ; Allow repeated abductions
  (Utilities|Array|Clear (Variables|Default|GetSingleAbductionCandidates))
  ;;[1 1 1] Abduct random player
  (if (Networking|IsServer)
    (for _array_element (Variables|/Script/Engine.GameStateBase|GetPlayerArray (Fortnite|Athena|GameState|GetGameStateAthena))
      (bind _asfort_player_state_athena (Utilities|Casting|CastToFortPlayerStateAthena _array_element))
      (bind _bdidactorhavetags (Class|AthenaQuestFunctionLibrary|AthenaCheckActorHasTags (PlayerState|GetCurrentPawn _asfort_player_state_athena) "(GameplayTags=((TagName=\"Athena.Gameplay.Status.AlreadyAbducted\")))" true))
      (if (Math|Boolean|ANDBoolean (not (Variables|/Script/Engine.PlayerState|GetbIsABot _asfort_player_state_athena)) (not (Health|IsDBNO (PlayerState|GetPlayerPawnorVehicleDriver _asfort_player_state_athena))) (not (.bInitialized (Utilities|Struct|BreakDeathInfo (Class|FortPlayerStateAthena|GetDeathInfo _asfort_player_state_athena)))) (not (select (Variables|Default|GetUsingSingleTestCheat) false _bdidactorhavetags)))
        (Utilities|Array|AddUnique (Variables|Default|GetSingleAbductionCandidates) (PlayerState|GetPlayerPawnorVehicleDriver _asfort_player_state_athena))))
    (if (> (Utilities|Array|Length (Variables|Default|GetSingleAbductionCandidates)) 0)
      (bind _output_get (Variables|Default|SetcurrentAbductionIndex _outindex))
      (bind _result (ArrayMacros|SafeArrayGet (Variables|Default|GetSingleAbductionCandidates) _output_get)
        (:%20%20
          (Variables|Default|SetSingleAbductee _result) ; Select random player
          (bind _returnvalue (Building|SpawnBuildingGameplayActor "/AwfulDuck/Items/AwfulDuck/BGA_AwfulDuck.BGA_AwfulDuck_C" (Math|Conversions|ToTransform(Vector) (Math|Vector|vector+vector (* (Math|Vector|VectorUp) 4096.0) (* (Transformation|GetActorForwardVector (Variables|Default|GetSingleAbductee)) 4096.0) (Transformation|GetActorLocation (Variables|Default|GetSingleAbductee)))) (Variables|Default|GetSingleAbductee)))
          (bind _asbga_awful_duck (Utilities|Casting|CastToBGA_AwfulDuck _returnvalue))
          (Class|BGAAwfulDuck|StartSingleAbduction _asbga_awful_duck (Variables|Default|GetSingleAbductee))))
      (else
        _))))

(event Custom|PhaseStepChanged
  (params (GamePhaseStep :type EAthenaGamePhaseStep%20Enum))
  (bind (_outitem _outindex) (Utilities|Array|RandomArrayItem (Variables|Default|GetSingleAbductionCandidates)))
  ;[1 1 1] Periodic abductions
  (if (ScalableFloat|GetValueasBool (Variables|Default|GetAwfulDuckEnabled))
    (if (Utilities|Enum|Equal(Enum) GamePhaseStep "StormHolding")
      (if (and (>= (SafeZone|GetCurrentPhaseIndex (SafeZone|GetSafeZoneIndicator)) (Variables|Default|GetfirstAbductionCircle)) (<= 0 (Variables|Default|GetlastAbductionCircle)))
        (bind _result (CallFunction|GetShouldDoRandomAbduct))
        (if _result
          (Utilities|Array|Clear (Variables|Default|GetSingleAbductionCandidates))
          ;;[1 1 1] Abduct random player
          (if (Networking|IsServer)
            (for _array_element (Variables|/Script/Engine.GameStateBase|GetPlayerArray (Fortnite|Athena|GameState|GetGameStateAthena))
              (bind _asfort_player_state_athena (Utilities|Casting|CastToFortPlayerStateAthena _array_element))
              (bind _bdidactorhavetags (Class|AthenaQuestFunctionLibrary|AthenaCheckActorHasTags (PlayerState|GetCurrentPawn _asfort_player_state_athena) "(GameplayTags=((TagName=\"Athena.Gameplay.Status.AlreadyAbducted\")))" true))
              (if (Math|Boolean|ANDBoolean (not (Variables|/Script/Engine.PlayerState|GetbIsABot _asfort_player_state_athena)) (not (Health|IsDBNO (PlayerState|GetPlayerPawnorVehicleDriver _asfort_player_state_athena))) (not (.bInitialized (Utilities|Struct|BreakDeathInfo (Class|FortPlayerStateAthena|GetDeathInfo _asfort_player_state_athena)))) (not (select (Variables|Default|GetUsingSingleTestCheat) false _bdidactorhavetags)))
                (Utilities|Array|AddUnique (Variables|Default|GetSingleAbductionCandidates) (PlayerState|GetPlayerPawnorVehicleDriver _asfort_player_state_athena))))
            (if (> (Utilities|Array|Length (Variables|Default|GetSingleAbductionCandidates)) 0)
              (bind _output_get (Variables|Default|SetcurrentAbductionIndex _outindex))
              (bind _result_2 (ArrayMacros|SafeArrayGet (Variables|Default|GetSingleAbductionCandidates) _output_get)
                (:%20%20
                  (Variables|Default|SetSingleAbductee _result_2) ; Select random player
                  (bind _returnvalue (Building|SpawnBuildingGameplayActor "/AwfulDuck/Items/AwfulDuck/BGA_AwfulDuck.BGA_AwfulDuck_C" (Math|Conversions|ToTransform(Vector) (Math|Vector|vector+vector (* (Math|Vector|VectorUp) 4096.0) (* (Transformation|GetActorForwardVector (Variables|Default|GetSingleAbductee)) 4096.0) (Transformation|GetActorLocation (Variables|Default|GetSingleAbductee)))) (Variables|Default|GetSingleAbductee)))
                  (bind _asbga_awful_duck (Utilities|Casting|CastToBGA_AwfulDuck _returnvalue))
                  (Class|BGAAwfulDuck|StartSingleAbduction _asbga_awful_duck (Variables|Default|GetSingleAbductee))))
              (else
                _))))))))

(event Custom|MassAbductHere
  ;[1 1 1] QA Testing cheats
  (Development|ExecuteConsoleCommand (Utilities|String|ToString(Text) (Utilities|Text|FormatText "INVTEXT(\"Cheat ke * ForceTestAbduction (X={x},Y={y},Z={z})\")" (Utilities|Text|ToText(Float) (.x (Transformation|GetActorLocation (Pawn|GetControlledPawn (select (Utilities|IsValid (Fortnite|GetFirstLocalFortPlayerController true) :!class /Script/Engine.KismetSystemLibrary) (Fortnite|GetFirstLocalFortPlayerController true) (Game|GetPlayerController 0))))) "HalfToEven" false false 1 324 0 0) (Utilities|Text|ToText(Float) (.y (Transformation|GetActorLocation (Pawn|GetControlledPawn (select (Utilities|IsValid (Fortnite|GetFirstLocalFortPlayerController true) :!class /Script/Engine.KismetSystemLibrary) (Fortnite|GetFirstLocalFortPlayerController true) (Game|GetPlayerController 0))))) "HalfToEven" false false 1 324 0 0) (Utilities|Text|ToText(Float) (.z (Transformation|GetActorLocation (Pawn|GetControlledPawn (select (Utilities|IsValid (Fortnite|GetFirstLocalFortPlayerController true) :!class /Script/Engine.KismetSystemLibrary) (Fortnite|GetFirstLocalFortPlayerController true) (Game|GetPlayerController 0))))) "HalfToEven" false false 1 324 0 0))) (Fortnite|GetFirstLocalFortPlayerController true)))

(event Custom|Begin%20Mass%20Abduction
  (bind _awfulduckselectbuffertime (Variables|Default|GetAwfulDuckSelectBufferTime))
  (bind _returnvalue (ScalableFloat|GetValueatLevel _awfulduckselectbuffertime))
  ;[1 1 1] Mass Abductions
  (bind _outactors (Actor|GetAllActorsOfClass "/AwfulDuck/Items/AwfulDuck/BGA_AwfulDuck.BGA_AwfulDuck_C"))
  (Variables|Default|SetAbductionSites _outactors)
  (CallFunction|UpdateCandidates)
  (for _array_element (Variables|Default|GetAbductionSites)
    (Class|BGAAwfulDuck|StartAbduction _array_element))
  (bind _returnvalue_2 (Utilities|Time|SetTimerbyEvent (delegate Custom|Update%20Abduction%20Candidates) _returnvalue true))
  (Utilities|Time|SetTimerbyEvent (delegate Custom|Warning%20Phase%20Finished) (- (ScalableFloat|GetValueatLevel (Variables|Default|GetAwfulDuckWarmUpTime)) _returnvalue)))

(event Custom|Update%20Abduction%20Candidates
  ;[1 1 1] Mass Abductions
  (CallFunction|UpdateCandidates))

(event Custom|Warning%20Phase%20Finished
  ;[1 1 1] Mass Abductions
  (Utilities|Time|ClearandInvalidateTimerbyHandle (stale Custom|Begin%20Mass%20Abduction _returnvalue_2))
  (CallFunction|ReportAnalytics))

(event Custom|ForceTestAbduction
  (params (Location :type Vector))
  ;[1 1 1] QA Testing cheats
  (if (Networking|IsServer)
    (Development|PrintString (Utilities|String|Append "CHEAT mass abduction at: " (Utilities|String|ToString(Vector) Location)))
    (Building|SpawnBuildingGameplayActor "/AwfulDuck/Items/AwfulDuck/BGA_AwfulDuck.BGA_AwfulDuck_C" (Math|Conversions|ToTransform(Vector) (+ (* (Math|Vector|VectorUp) 4096.0) Location)) self)
    (CallFunction|BeginMassAbduction)))

(event EventEndPlay
  (params (EndPlayReason :type EEndPlayReason%20Enum :default "Destroyed"))
  ;[1 0.8418 0.5363] End play - unbind delegates
  (|CallParentFunction EndPlayReason)
  (bind _gamephaseupdatedasynctask (Variables|Default|GetGamePhaseUpdatedAsyncTask)
    (:then
      (GameplayEventRouter|StopListeningtoStatefulEvent _gamephaseupdatedasynctask))
    (:else
      _))
  (bind _gamephasestepupdatedasynctask (Variables|Default|GetGamePhaseStepUpdatedAsyncTask)
    (:then
      (GameplayEventRouter|StopListeningtoStatefulEvent _gamephasestepupdatedasynctask))
    (:else
      _))
  (bind _creativestormshieldgamephaseupdatedasynctask (Variables|Default|GetCreativeStormShieldGamePhaseUpdatedAsyncTask)
    (:then
      (GameplayEventRouter|StopListeningtoStatefulEvent _creativestormshieldgamephaseupdatedasynctask))
    (:else)))

(event Game|Damage|EventAnyDamage
  (params (Damage :type Float%20%28single-precision%29) (DamageType :type Damage%20Type%20Object%20Reference) (InstigatedBy :type Controller%20Object%20Reference) (DamageCauser :type Actor%20Object%20Reference)))

(fn ArrayDifference
  (params (Array%20A :type Array%20of%20Strings) (Array%20B :type Array%20of%20Strings))
  (for _array_element Array%20A
    (if (not (Utilities|Array|ContainsItem Array%20B _array_element))
      (Utilities|Array|AddUnique (LocalVariables|GetResultArray) _array_element)))
  (return (LocalVariables|GetResultArray)))

(fn BroadcastAbductees
  (params (Abductees :type Array%20of%20Awful%20Duck%20Abduction%20Group%20Structures))
  (for _array_element Abductees
    (Variables|/AwfulDuck/Items/AwfulDuck/BGA_AwfulDuck.BGA_AwfulDuck_C|SetAbductees (.Abductees_8_143E977A49232B524E57AE8BDEB7BBC4 (Utilities|Struct|BreakAwfulDuckAbductionGroup _array_element)) (Utilities|Struct|BreakAwfulDuckAbductionGroup _array_element))))

(fn CountCandidates
  (params (Group :type Array%20of%20Awful%20Duck%20Abduction%20Group%20Structures))
  (for _array_element Group
    (LocalVariables|SetCount (+ (LocalVariables|GetCount) (Utilities|Array|Length (.Abductees_8_143E977A49232B524E57AE8BDEB7BBC4 (Utilities|Struct|BreakAwfulDuckAbductionGroup _array_element))))))
  (return (LocalVariables|GetCount)))

(fn CullPlayersbySquad
  (params (AbductionGroups :type Array%20of%20Awful%20Duck%20Abduction%20Group%20Structures) (PlayerOfSquadToRemove :type Fort%20Player%20Pawn%20Athena%20Object%20Reference))
  (bind _workinggroup (LocalVariables|GetWorkingGroup))
  (LocalVariables|SetAbducteesRemaining 0) ; Count the number of abductees NOT\nculled as we cull the selected squad so\nwe do not need to recount afterwards.
  (for _array_element AbductionGroups
    (bind _returnvalue (Utilities|Array|AddUnique _workinggroup _array_element)) ; Add site
    (if (Utilities|Array|IsValidIndex _workinggroup _returnvalue)
      (Utilities|Array|Clear (.Abductees_8_143E977A49232B524E57AE8BDEB7BBC4 (Utilities|Struct|BreakAwfulDuckAbductionGroup (Utilities|Array|Get(aref) _workinggroup _returnvalue))))
      (for _array_element_2 (.Abductees_8_143E977A49232B524E57AE8BDEB7BBC4 (Utilities|Struct|BreakAwfulDuckAbductionGroup _array_element))
        (bind _asfort_player_pawn_athena (Utilities|Casting|CastToFortPlayerPawnAthena _array_element_2))
        (if (not (Squad|OnSameSquad PlayerOfSquadToRemove _asfort_player_pawn_athena))
          (Utilities|Array|AddUnique (.Abductees_8_143E977A49232B524E57AE8BDEB7BBC4 (Utilities|Struct|BreakAwfulDuckAbductionGroup (Utilities|Array|Get(aref) (LocalVariables|GetWorkingGroup) _returnvalue))) _asfort_player_pawn_athena) ; You can stay
          (Math|Integer|IncrementInt (LocalVariables|GetAbducteesRemaining))))))
  (return (LocalVariables|GetWorkingGroup) (LocalVariables|GetAbducteesRemaining)))

(fn FilterforFortPlayerPawnAthena
  (params (Actors :type Array%20of%20Actor%20Object%20References))
  (for _array_element Actors
    (bind _asfort_player_pawn_athena (Utilities|Casting|CastToFortPlayerPawnAthena _array_element))
    (Utilities|Array|AddUnique (LocalVariables|GetFortAthenaPawns) _asfort_player_pawn_athena))
  (return (LocalVariables|GetFortAthenaPawns)))

(fn GetLastValidPlayerinGroup
  (params (Group :type Array%20of%20Awful%20Duck%20Abduction%20Group%20Structures))
  (for _array_element Group
    (bind (_arrayindex _arrayelement) (Utilities|Array|ReverseforEachLoop (.Abductees_8_143E977A49232B524E57AE8BDEB7BBC4 (Utilities|Struct|BreakAwfulDuckAbductionGroup _array_element)))
      (:LoopBody
        (Utilities|IsValid _arrayelement
          (:Is%20Valid
            (return _arrayelement true))))))
  (return 0 false))

(fn GetShouldDoRandomAbduct
  (bind _maxabductionlevels (Variables|Default|GetMaxAbductionLevels))
  (bind _returnvalue (Math|Conversions|ToFloat(Integer) _maxabductionlevels))
  (bind _returnvalue_2 (Math|Random|RandomFloat))
  (bind _awfulduckrandomtargetchance (Variables|Default|GetAwfulDuckRandomTargetChance))
  (bind (_biseventactive _timesincebegin _timeuntilend _timespanratio) (CalendarEvents|GetCalendarEventInformation (Variables|Default|GetForeshadowingEventName)))
  (if _biseventactive
    (Development|PrintString (Utilities|String|Append "Random abduct (lvl  " (Utilities|String|ToString(Float) _timespanratio) "):" (Utilities|String|ToString(Float) _returnvalue_2) "/" (Utilities|String|ToString(Float) (ScalableFloat|GetValueatLevel _awfulduckrandomtargetchance (* _timespanratio _returnvalue)))))
    (return (<= _returnvalue_2 (ScalableFloat|GetValueatLevel _awfulduckrandomtargetchance (* _timespanratio _returnvalue))))
    (else
      (return false))))

(fn GetSkyfireDestination
  (params (Pawn :type Fort%20Player%20Pawn%20Object%20Reference))
  (bind _squadids (Variables|Default|GetSquadIds))
  (bind _asfort_player_pawn_athena (Utilities|Casting|CastToFortPlayerPawnAthena Pawn)
    (:then
      (bind _asfort_player_state_athena (Utilities|Casting|CastToFortPlayerStateAthena (Variables|/Script/Engine.Pawn|GetPlayerState _asfort_player_pawn_athena))
        (:then
          (if (> (Utilities|Array|FindItem _squadids (Class|FortPlayerStateAthena|GetSquadId _asfort_player_state_athena)) -1)
            (if (Utilities|Array|IsValidIndex (Variables|Default|GetSkyfireStartLocations) (Utilities|Array|FindItem _squadids (Class|FortPlayerStateAthena|GetSquadId _asfort_player_state_athena)))
              (return (FortSquadStart|GetPlayerStartTransform (Utilities|Array|Get(acopy) (Variables|Default|GetSkyfireStartLocations) (Utilities|Array|FindItem _squadids (Class|FortPlayerStateAthena|GetSquadId _asfort_player_state_athena))) (Class|FortPlayerStateAthena|GetSquadMemberIndex _asfort_player_state_athena)))
              (else
                (Development|PrintString "We have more squads than we do destination points")
                (return (Math|Transform|MakeTransform "7000.000000, 0,20000.000000"))))
            (else
              (Development|PrintString "We don't have this player's squad id in our abduction list"))))
        (:CastFailed
          _)))
    (:CastFailed
      _))
  (return (Math|Transform|MakeTransform "7000.000000, 0,20000.000000")))

(fn GetUniqueCandidateIDs
  (params (Pawns :type Array%20of%20Actor%20Object%20References))
  (for _array_element Pawns
    (bind _asfort_player_pawn_athena (Utilities|Casting|CastToFortPlayerPawnAthena _array_element))
    (bind _asfort_player_state_athena (Utilities|Casting|CastToFortPlayerStateAthena (Variables|/Script/Engine.Pawn|GetPlayerState _asfort_player_pawn_athena)))
    (bind _returnvalue (PlayerInfo|GetDebugStringforUniqueId (PlayerState|GetPlayerUniqueNetID _asfort_player_state_athena)))
    (Utilities|Array|AddUnique (LocalVariables|GetIDArray) _returnvalue))
  (return (LocalVariables|GetIDArray)))

(fn InitFromRows
  (Variables|Default|SetfirstAbductionCircle (Math|Float|Truncate (ScalableFloat|GetValueatLevel (Variables|Default|GetAwfulDuckStartCircle))))
  (bind (_biseventactive _timesincebegin _timeuntilend _timespanratio) (CalendarEvents|GetCalendarEventInformation (Variables|Default|GetForeshadowingEventName)))
  (Variables|Default|SetlastAbductionCircle (Math|Integer|SelectInt (Math|Float|Truncate (ScalableFloat|GetValueatLevel (Variables|Default|GetAwfulDuckRandomCount) (* _timespanratio (Math|Conversions|ToFloat(Integer) (Variables|Default|GetMaxAbductionLevels))))) (Math|Float|Truncate (ScalableFloat|GetValueatLevel (Variables|Default|GetAwfulDuckEndCircle))) _biseventactive)))

(fn InitSkyfireDestinations
  (if (Networking|IsServer)
    (bind _outactors (Actor|GetAllActorsOfClass "/Script/FortniteGame.FortSquadStart"))
    (for _array_element _outactors
      (if (GameplayTags|HasTag (FortSquadStart|GetSquadStartGameplayTags _array_element) (LocalVariables|GetSkyfireStartLocationTag) true)
        (Utilities|Array|AddUnique (LocalVariables|GetTempLocations) _array_element)))
    (Variables|Default|SetSkyfireStartLocations (LocalVariables|GetTempLocations))))

(fn ReportAnalytics
  (bind _returnvalue (SafeZone|GetSafeZoneIndicator))
  (bind _returnvalue_2 (Fortnite|Athena|GameState|GetGameStateAthena))
  (bind _returnvalue_3 (Fortnite|Athena|GameState|GetGameStateAthena))
  (if (and (Networking|IsServer) (not (Replay|GetIsPlayingReplay)))
    (Utilities|FlowControl|Sequence
      (:then_0
        ;[1 1 1] Compile abduction locations and players
        (for _array_element (Variables|Default|GetSkyfireAbductees)
          (Utilities|Array|AddUnique (LocalVariables|GetAbductionLocations) (Variables|/AwfulDuck/Items/AwfulDuck/BGA_AwfulDuck.BGA_AwfulDuck_C|GetLocationCode (Utilities|Struct|BreakAwfulDuckAbductionGroup _array_element)))
          (bind _ids (CallFunction|GetUniqueCandidateIDs (.Abductees_8_143E977A49232B524E57AE8BDEB7BBC4 (Utilities|Struct|BreakAwfulDuckAbductionGroup _array_element))))
          (LocalVariables|SetAbductedPlayers _ids)))
      (:then_1
        ;[1 1 1] Compile rejected and aborted players
        (for _array_element_2 (Variables|Default|GetSkyfireAbductionCandidates)
          (bind _ids_2 (CallFunction|GetUniqueCandidateIDs (.Abductees_8_143E977A49232B524E57AE8BDEB7BBC4 (Utilities|Struct|BreakAwfulDuckAbductionGroup _array_element_2))))
          (Utilities|Array|AppendArray (LocalVariables|GetCurrentCandidateIDs) _ids_2))
        (LocalVariables|SetRejectedPlayers (CallFunction|ArrayDifference (LocalVariables|GetCurrentCandidateIDs) (LocalVariables|GetAbductedPlayers)))
        (LocalVariables|SetAbortedPlayers (CallFunction|ArrayDifference (Variables|Default|GetAllCandidateIDs) (LocalVariables|GetCurrentCandidateIDs))))
      (:then_2
        (bind _returnvalue_4 (Analytics|ConstructAnalyticsAttribute "GameSessionID" (Session|GetGameSessionID _returnvalue_3)))
        (bind _returnvalue_5 (Analytics|ConstructAnalyticsAttribute "PlaylistName" (Utilities|String|ToString(Name) (Playlist|GetCurrentPlaylistName _returnvalue_3))))
        (bind _returnvalue_6 (Analytics|ConstructAnalyticsAttribute :AttributeName "MatchTime" :IntValue -1 :floatValue (GameState|GetMatchTime _returnvalue_3) :BoolValue false))
        (if (Utilities|IsValid _returnvalue :!class /Script/Engine.KismetSystemLibrary)
          (bind _returnvalue_7 (Analytics|ConstructAnalyticsAttribute "StormCenter" (CallFunction|VectortoIntString (SafeZone|GetSafeZoneCenter _returnvalue))))
          (bind _returnvalue_8 (Analytics|ConstructAnalyticsAttribute :AttributeName "StormRadius" :IntValue -1 :floatValue (Variables|/Script/FortniteGame.FortSafeZoneIndicator|GetRadius _returnvalue) :BoolValue false)))
        (bind _returnvalue_9 (Analytics|ConstructAnalyticsAttribute "AbductingPOICodes" (Utilities|String|JoinStringArray (LocalVariables|GetAbductionLocations) ",")))
        (bind _returnvalue_10 (Analytics|ConstructAnalyticsAttribute :AttributeName "PlayersAlive" :IntValue (Variables|/Script/FortniteGame.FortGameStateAthena|GetPlayersLeft _returnvalue_2) :floatValue -1.0 :BoolValue false))
        (bind _returnvalue_11 (Analytics|ConstructAnalyticsAttribute :AttributeName "TeamsAlive" :IntValue (Variables|/Script/FortniteGame.FortGameStateAthena|GetTeamsLeft _returnvalue_2) :floatValue -1.0 :BoolValue false))
        (bind _returnvalue_12 (Analytics|ConstructAnalyticsAttribute "PlayersAbducted" (Utilities|String|JoinStringArray (LocalVariables|GetAbductedPlayers) ",")))
        (bind _returnvalue_13 (Analytics|ConstructAnalyticsAttribute "PlayersRejected" (Utilities|String|JoinStringArray (LocalVariables|GetRejectedPlayers) ",")))
        (bind _returnvalue_14 (Analytics|ConstructAnalyticsAttribute "PlayersAborted" (Utilities|String|JoinStringArray (LocalVariables|GetAbortedPlayers) ","))))
      (:then_3
        (Analytics|FireAnalyticsEventBlueprintArray "Core.AthenaAbductionEvent" (Utilities|Array|MakeArray _returnvalue_4 _returnvalue_5 _returnvalue_6 _returnvalue_7 _returnvalue_8 _returnvalue_9 _returnvalue_10 _returnvalue_11 _returnvalue_12 _returnvalue_13 _returnvalue_14)))
      (:then_4
        (Development|PrintString (Utilities|String|JoinStringArray (LocalVariables|GetAbortedPlayers) ","))))))

(fn SaveandBroadcastSquadIds
  (params (Abductees :type Array%20of%20Awful%20Duck%20Abduction%20Group%20Structures))
  (for _array_element Abductees
    (for _array_element_2 (.Abductees_8_143E977A49232B524E57AE8BDEB7BBC4 (Utilities|Struct|BreakAwfulDuckAbductionGroup _array_element))
      (bind _asfort_player_pawn_athena (Utilities|Casting|CastToFortPlayerPawnAthena _array_element_2))
      (bind _asfort_player_state_athena (Utilities|Casting|CastToFortPlayerStateAthena (Variables|/Script/Engine.Pawn|GetPlayerState _asfort_player_pawn_athena)))
      (Utilities|Array|AddUnique (LocalVariables|GetTempSquads) (Class|FortPlayerStateAthena|GetSquadId _asfort_player_state_athena))))
  (Variables|Default|SetSquadIds (LocalVariables|GetTempSquads)) ; All done. Send it out to the clients.
)

(fn SelectAbductees
  (params (AbductionSiteGroups :type Array%20of%20Awful%20Duck%20Abduction%20Group%20Structures))
  (bind _numfilteredplayers (LocalVariables|GetNumFilteredPlayers))
  (bind _returnvalue (Utilities|String|ToString(Integer) _numfilteredplayers))
  (bind _numhumans (LocalVariables|GetNumHumans))
  (bind _returnvalue_2 (Utilities|String|ToString(Integer) _numhumans))
  (bind _curnumhumans (LocalVariables|GetCurNumHumans))
  (bind _numfilteredplayers_2 (LocalVariables|GetNumFilteredPlayers))
  (bind _returnvalue_3 (- _numfilteredplayers_2 _curnumhumans))
  (bind _returnvalue_4 (Utilities|String|ToString(Integer) _returnvalue_3))
  (bind _numbots (LocalVariables|GetNumBots))
  (bind _returnvalue_5 (Utilities|String|ToString(Integer) _numbots))
  (bind _curnumhumans_2 (LocalVariables|GetCurNumHumans))
  (bind _returnvalue_6 (Utilities|String|ToString(Integer) _curnumhumans_2))
  (bind _maxabductees (LocalVariables|GetMaxAbductees))
  (bind _returnvalue_7 (Utilities|String|ToString(Integer) _maxabductees))
  (bind _numcullloops (LocalVariables|GetNumCullLoops))
  (bind _returnvalue_8 (Utilities|String|ToString(Integer) _numcullloops))
  (bind _maxcullloops (LocalVariables|GetMaxCullLoops))
  (bind _returnvalue_9 (Utilities|String|ToString(Integer) _maxcullloops))
  (bind _returnvalue_10 (Utilities|String|Append "Select Abductees: " _returnvalue "/" _returnvalue_7 " ( Humans: " _returnvalue_6 "/" _returnvalue_2 " , Bots: " _returnvalue_4 "/" _returnvalue_5 " , Cull Loops: " _returnvalue_8 "/" _returnvalue_9 " )"))
  (bind _output_get (LocalVariables|SetMaxAbductees (Math|Float|Truncate (ScalableFloat|GetValueatLevel (Variables|Default|GetAwfulDuckMaxCapacity)))))
  (bind _output_get_2 (LocalVariables|SetTotalAbductionCandidates (CallFunction|CountCandidates AbductionSiteGroups)))
  (if (<= _output_get_2 _output_get)
    (Fortnite|FortShippingLog self (Utilities|String|Append "Found " (Utilities|String|ToString(Integer) _output_get_2) " abduction candidates. No filtering required."))
    ;[1 0.25 0.5] Under limit, no filtering needed - Return
    (return (Variables|GetAbductionSiteGroups))
    (else
      ;[0.1362 0.7682 1] Seperate humans and bots
      (LocalVariables|SetNumBots 0)
      (LocalVariables|SetNumHumans 0)
      (for (_array_element _array_index) (Variables|GetAbductionSiteGroups)
        (Utilities|Array|Add (LocalVariables|GetHumanPlayers) (Utilities|Struct|MakeAwfulDuckAbductionGroup (Utilities|Struct|BreakAwfulDuckAbductionGroup _array_element))) ; Add site to new list
        (Utilities|Array|Add (LocalVariables|GetBotPlayers) (Utilities|Struct|MakeAwfulDuckAbductionGroup (Utilities|Struct|BreakAwfulDuckAbductionGroup _array_element))) ; Add new site to list
        (for _array_element_2 (.Abductees_8_143E977A49232B524E57AE8BDEB7BBC4 (Utilities|Struct|BreakAwfulDuckAbductionGroup _array_element))
          (bind _asfort_player_pawn_athena (Utilities|Casting|CastToFortPlayerPawnAthena _array_element_2))
          (bind _asfort_player_state_athena (Utilities|Casting|CastToFortPlayerStateAthena (Variables|/Script/Engine.Pawn|GetPlayerState _asfort_player_pawn_athena)))
          (if (Variables|/Script/Engine.PlayerState|GetbIsABot _asfort_player_state_athena)
            (Utilities|Array|AddUnique (.Abductees_8_143E977A49232B524E57AE8BDEB7BBC4 (Utilities|Struct|BreakAwfulDuckAbductionGroup (Utilities|Array|Get(aref) (LocalVariables|GetBotPlayers) _array_index))) _array_element_2) ; Add individual to bot list
            (Math|Integer|IncrementInt (LocalVariables|GetNumBots))
            (else
              (Utilities|Array|AddUnique (.Abductees_8_143E977A49232B524E57AE8BDEB7BBC4 (Utilities|Struct|BreakAwfulDuckAbductionGroup (Utilities|Array|Get(aref) (LocalVariables|GetHumanPlayers) _array_index))) _array_element_2) ; Add individual to player list
              (Math|Integer|IncrementInt (LocalVariables|GetNumHumans))))))
      ;[0.4541 1 0.5625] Put all the humans in
      (LocalVariables|SetFilteredPlayers (LocalVariables|GetHumanPlayers)) ; Put all the humans in
      (LocalVariables|SetNumFilteredPlayers (LocalVariables|GetNumHumans))
      ;[0.5273 0.8758 1] Initialize Max/Num Cull Loop Variables (initialize here so they can be used in logging even if we don't cull any players)
      (LocalVariables|SetMaxCullLoops (Math|Integer|Min(Integer) (LocalVariables|GetNumFilteredPlayers) 100)) ; Safety limit for max while loops
      (LocalVariables|SetNumCullLoops 0)
      (if (not (<= (LocalVariables|GetNumFilteredPlayers) (LocalVariables|GetMaxAbductees)))
        ;[1 0.7771 0.4983] Over limit with just humans, cull players by squad
        (while (and (> (LocalVariables|GetNumFilteredPlayers) (LocalVariables|GetMaxAbductees)) (<= (LocalVariables|GetNumCullLoops) (LocalVariables|GetMaxCullLoops)))
          (bind (_player _success) (CallFunction|GetLastValidPlayerInGroup (LocalVariables|GetFilteredPlayers)))
          (if _success
            (bind _asfort_player_pawn_athena_2 (Utilities|Casting|CastToFortPlayerPawnAthena _player)
              (:then
                (bind (_outabductiongroups _outnumabducteesremaining) (CallFunction|CullPlayersBySquad (LocalVariables|GetFilteredPlayers) _asfort_player_pawn_athena_2)) ; Cull this player's squad
                (LocalVariables|SetFilteredPlayers _outabductiongroups)
                (LocalVariables|SetNumFilteredPlayers _outnumabducteesremaining)
                (Math|Integer|IncrementInt (LocalVariables|GetNumCullLoops)))
              (:CastFailed
                ;;[1 0.25 0.5] Return - Error cases
                (return 0)))
            (else
              (return 0)))))
      ;[0.3227 1 0.5672] Under or exact with just humans. Add bots if possible.
      (LocalVariables|SetCurNumHumans (LocalVariables|GetNumFilteredPlayers))
      (bind (_array_element_3 _array_index_2) (Utilities|Array|ForEachLoopwithBreak (LocalVariables|GetBotPlayers))
        (:LoopBody
          (for _array_element_4 (.Abductees_8_143E977A49232B524E57AE8BDEB7BBC4 (Utilities|Struct|BreakAwfulDuckAbductionGroup _array_element_3))
            (if (== (LocalVariables|GetNumFilteredPlayers) (LocalVariables|GetMaxAbductees))
              _)
            (Utilities|Array|AddUnique (.Abductees_8_143E977A49232B524E57AE8BDEB7BBC4 (Utilities|Struct|BreakAwfulDuckAbductionGroup (Utilities|Array|Get(aref) (LocalVariables|GetFilteredPlayers) _array_index_2))) _array_element_4) ; Add a bot from this site
            (Math|Integer|IncrementInt (LocalVariables|GetNumFilteredPlayers))))
        (:Completed
          _))
      ;[1 0.3802 0.664] Log Results and Return
      (Fortnite|FortShippingLog self _returnvalue_10)
      (Development|PrintString _returnvalue_10 true false)
      (return (LocalVariables|GetFilteredPlayers)))))

(fn UpdateCandidates
  (bind _skyfireabductees (Variables|Default|GetSkyfireAbductees))
  (Utilities|Array|Clear (Variables|Default|GetSkyfireAbductionCandidates))
  (for _array_element (Variables|Default|GetAbductionSites)
    (bind _result (Class|BGAAwfulDuck|GetEligibleOverlappingPawns _array_element))
    (Utilities|Array|AddUnique (Variables|Default|GetSkyfireAbductionCandidates) (Utilities|Struct|MakeAwfulDuckAbductionGroup _array_element (CallFunction|FilterForFortPlayerPawnAthena _result)))
    (bind _ids (CallFunction|GetUniqueCandidateIDs (CallFunction|FilterForFortPlayerPawnAthena _result)))
    (for _array_element_2 _ids
      (Utilities|Array|AddUnique (Variables|Default|GetAllCandidateIDs) _array_element_2)))
  (bind _result_2 (CallFunction|SelectAbductees (Variables|Default|GetSkyfireAbductionCandidates))) ; Select who is actually going to get abducted, obeying quota
  (bind _output_get (Variables|Default|SetSkyfireAbductees _result_2))
  (CallFunction|SaveAndBroadcastSquadIds _output_get) ; Save squad ids for teleport
  (CallFunction|BroadcastAbductees _skyfireabductees) ; Tell individual BGAs who is getting abducted
  (Development|PrintString (Utilities|String|Append "Updating Skyfire abductees: " (Utilities|String|ToString(Integer) (CallFunction|CountCandidates _skyfireabductees)) " found")))

(fn ConstructionScript)

(fn VectorToIntString
  (params (InVec :type Vector))
  (return (Utilities|String|Append "(X=" (Utilities|String|ToString(Integer) (Math|Float|Floor (.x InVec))) ", Y=" (Utilities|String|ToString(Integer) (Math|Float|Floor (.y InVec))) ", Z=" (Utilities|String|ToString(Integer) (Math|Float|Floor (.z InVec))) ")")))
