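; AwfulDuck EQS context, named as a return point, that supplies a single location for environment queries.
; It overrides ProvideSingleLocation, stores the querier actor's world location in a center point variable, and returns that point plus an offset variable.
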(fn ProvideSingleLocation
  (params (QuerierObject :type Object%20Reference) (QuerierActor :type Actor%20Object%20Reference))
  (LocalVariables|Setcenterpoint (Transformation|GetActorLocation QuerierActor))
  (return (+ (LocalVariables|Getcenterpoint) (LocalVariables|Getoffset))))
