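# Environment query that picks a placement point for returning the Awful Duck item, based on its return context.
# A large, coarse grid around that context projects down, excluding water through a nav filter.
# Tests keep points within a 2D distance band, favoring farther ones, score flat ground and weight safe zone placement more heavily.
# Filters also bound distance from world origin and reject kill volumes.
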EnvQuery /AwfulDuck/Items/AwfulDuck/EQS_AwfulDuckReturnPlacement
options [{
  generator {
   name=EnvQueryGenerator_SimpleGrid_1
   class=EnvQueryGenerator_SimpleGrid
   properties {
    generateAround=&/EQSC_AwfulDuck_Return.EQSC_AwfulDuck_Return_C
    gridSize=25000
    spaceBetween=4000
    projectionData {
     extentX=0
     navigationFilter=&NavFilter_NPC_ExcludeAllWater_C
     postProjectionVerticalOffset=0
     projectDown=20000
     projectUp=0
     serializedChannel=ECC_GameTraceChannel6
     traceChannel=TraceTypeQuery7
    }
   }
  }
  tests [
   EnvQueryTest_Distance_1 {
    class=EnvQueryTest_Distance
    properties {
     distanceTo=&EQSC_AwfulDuck_Return_C
     filterType=Range
     floatValueMax=30000
     floatValueMin=4000
     scoringEquation=Linear
     scoringFactor=1
     testComment=""
     testMode=Distance2D
     testPurpose=FilterAndScore
    }
   }
   FortQueryTest_OnFlatSurface_0 {
    class=FortQueryTest_OnFlatSurface
    properties {
     radius=50
     toleranceZ=10
     traceOffsetUp=100
     scoringFactor {
      dataBinding=&self:FortQueryTest_OnFlatSurface_0.AIDataProvider_QueryParams_1
      dataField=FloatValue
      defaultValue=0.5
     }
    }
   }
   FortQueryTest_InsideAthenaSafeZone_0 {class=FortQueryTest_InsideAthenaSafeZone, properties {scoringFactor=3, testPurpose=Score}}
   EnvQueryTest_Distance_0 {
    class=EnvQueryTest_Distance
    properties {
     distanceTo=/Game/Missions/Secondary/Events/SatelliteSalvage/Context_WorldOrigin.Context_WorldOrigin_C
     filterType=Range
     floatValueMax=148000
     floatValueMin=0
     scoringEquation=Linear
     scoringFactor=1
     testComment=""
     testMode=Distance2D
     testPurpose=Filter
    }
   }
   EnvQueryTest_Volume_0 {class=EnvQueryTest_Volume, properties {boolValue=false, testPurpose=Filter, volumeClass=/Script/Engine.KillZVolume, volumeContext=null}}
  ]
 }]
