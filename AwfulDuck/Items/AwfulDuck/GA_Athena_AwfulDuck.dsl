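; Gameplay ability for the AwfulDuck abduction, which beams the player up and teleports them to the ground or the mothership.
; After a delay it clears blocking states like vehicles, DBNO and carrying, then lifts the pawn with root motion.
; Single abductions teleport to an EQS-picked spot and restore health and shield; others recheck abort conditions and use the manager's Skyfire destination.
; Ending cleans up effects, tags uncancelled abductees and grants fall immunity if falling.
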(event Ability|EventActivateAbility
  ;[1 1 1] Init vars
  (bind _output_get (Variables|Default|SetFortPawn (Utilities|Casting|CastToFortPlayerPawn (Ability|GetAvatarActorfromActorInfo))))
  (Ability|SendGameplayEventtoActor _output_get "(TagName=\"Event.Athena.Item.AwfulDuck.Abduct.Start\")")
  (Variables|Default|SetInstigatorBGA (Ability|EffectContext|GetInstigatorActor (Ability|GetGrantedbyEffectContext)))
  (bind _singleabduction (CallFunction|IsSingleAbduction))
  (Variables|Default|SetSingle%20Abduction _singleabduction)
  (bind _returnvalue (Actor|GetActorOfClass "/AwfulDuck/Items/AwfulDuck/B_AwfulDuck_Manager.B_AwfulDuck_Manager_C"))
  (Variables|Default|SetAbductionManager _returnvalue)
  ;[1 1 1] Abduction effects
  (Utilities|Time|SetTimerbyEvent (delegate Custom|TimerFinished0) 1.0) ; Wait 1 second
)

(event Custom|TimerFinished0
  (bind _returnvalue (Ability|GetAvatarActorfromActorInfo))
  (bind _instigatorbga (Variables|Default|GetInstigatorBGA))
  (bind _returnvalue_2 (Transformation|GetActorLocation _instigatorbga))
  (bind _returnvalue_3 (Ability|GetGrantedbyEffectContext))
  (bind _returnvalue_4 (Ability|GameplayCue|MakeGameplayCueParameters :NormalizedMagnitude 1.0 :RawMagnitude 1.0 :EffectContext _returnvalue_3 :Location _returnvalue_2 :Normal "0,0,0" :EffectCauser _returnvalue :GameplayEffectLevel 1 :AbilityLevel 1 :bReplicateLocationWhenUsingMinimalRepProxy true))
  (bind _delayafterteleport (Variables|Default|GetDelayAfterTeleport))
  (bind _returnvalue_5 (ScalableFloat|GetValueatLevel _delayafterteleport))
  (bind _fortpawn (Variables|Default|GetFortPawn))
  (bind _returnvalue_6 (Class|FortPlayerPawn|GetDBNOHoister _fortpawn))
  (bind _beamlifttime (Variables|Default|GetBeamLiftTime))
  (bind _returnvalue_7 (+ _beamlifttime _returnvalue_5))
  (bind _fortpawn_2 (Variables|Default|GetFortPawn))
  (bind _fortpawn_3 (Variables|Default|GetFortPawn))
  (bind _fortpawn_4 (Variables|Default|GetFortPawn))
  (bind _fortpawn_5 (Variables|Default|GetFortPawn))
  (bind _fortpawn_6 (Variables|Default|GetFortPawn))
  (bind _fortpawn_7 (Variables|Default|GetFortPawn))
  ;[1 1 1] Abduction effects
  (Utilities|FlowControl|Sequence
    (:then_0
      ;;[1 1 1] Carrying someone? Drop them.
      (bind _bdidactorhavetags (Class|AthenaQuestFunctionLibrary|AthenaCheckActorHasTags _fortpawn_4 "(GameplayTags=((TagName=\"Gameplay.Action.Player.CarryingPlayer\")))" true))
      (if _bdidactorhavetags
        (Class|FortPlayerPawn|StopCarryingPlayer _fortpawn_4 true)))
    (:then_1
      ;;[1 1 1] Eject from vehicle
      (if (Vehicle|IsinVehicle _fortpawn_2)
        (Vehicle|EjectPlayerPawn(Message) (Vehicle|GetVehicleActor _fortpawn_2 :!class /Script/FortniteGame.FortPlayerPawn) _fortpawn_2) ; Pull them out
        (else
          _)))
    (:then_2
      ;;[1 1 1] Revive if DBNO
      (Utilities|IsServer
        (:true
          (if (Health|IsDBNO _fortpawn_3)
            (bind _asfort_player_pawn_athena (Utilities|Casting|CastToFortPlayerPawnAthena _fortpawn_3))
            (Class|FortPlayerPawnAthena|ForceRevivefromDBNO _asfort_player_pawn_athena (Pawn|GetController _fortpawn_3))))))
    (:then_3
      ;;[1 1 1] Someone carrying you? Force drop
      (Utilities|IsValid _returnvalue_6
        (:Is%20Valid
          (bind _bdidactorhavetags_2 (Class|AthenaQuestFunctionLibrary|AthenaCheckActorHasTags _returnvalue_6 "(GameplayTags=((TagName=\"Gameplay.Action.Player.CarryingPlayer\")))" true))
          (if _bdidactorhavetags_2
            (Class|FortPlayerPawn|StopCarryingPlayer _returnvalue_6 true)))))
    (:then_4
      ;;[1 1 1] Force exit hiding prop
      (if (Class|FortPawn|IsInsideHidingProp _fortpawn_7 false)
        (Ability|SendGameplayEventtoActor _fortpawn_7 "(TagName=\"Event.Athena.HidingProp.StopHiding\")")))
    (:then_5
      ;;[1 1 1] Drop held objects
      (Utilities|IsServer
        (:true
          (bind _returnvalue_8 (Fortnite|HeldObjects|GetHeldObjectComponentAttachedtoPlayer _fortpawn_6))
          (if (Class|FortPawn|GetIsHoldingObject _fortpawn_6)
            (HeldObjects|Behavior|DropHeldObject _returnvalue_8)))))
    (:then_6
      (Utilities|FlowControl|Sequence
        (:then_0
          ;;[1 1 1] Exit zipline
          (bind _returnvalue_9 (Zipline|IsZiplining _fortpawn_5))
          (if _returnvalue_9
            (Zipline|EndZiplining _fortpawn_5 false false)))
        (:then_1
          ;;[1 1 1] Exit bush
          (Ability|SendGameplayEventtoActor (Variables|Default|GetFortPawn) "(TagName=\"Abilities.Generic.Athena.Bush.Destroyed\")"))))
    (:then_7
      (bind _returnvalue_10 (CallFunction|IsPlayerDead))
      (if _returnvalue_10
        (Development|PrintString "Ending ability early")
        (Ability|CancelAbility)
        (else
          (Development|PrintString "Eject check complete")
          (Variables|Default|SetNewVar_0 (Math|DateTime|UTCNow))
          ;;[1 1 1] Lifting beam
          (Fortnite|ApplyGameplayEffectsToActor (Variables|Default|GetFortPawn) (Utilities|Array|MakeArray (Variables|Default|GetLiftGE)) (Transformation|GetActorLocation (Variables|Default|GetFortPawn))) ; Make abductee invulnerable
          (Ability|AddGameplayCueWithParamsToOwner (Variables|Default|GetGC_Beam_Tag) _returnvalue_4 false) ; Lift beam
          (Ability|Tasks|PlayMontageAndWait "None" "/Game/Animation/Game/MainPlayer/Abducted/Abducted_Intro_Montage.Abducted_Intro_Montage" 1.0 "" false)
          (Ability|AddGameplayCueWithParamsToOwner (Variables|Default|GetRezOutAndIn_Tag) _returnvalue_4) ; Lift beam
          (Utilities|Time|SetTimerbyEvent (delegate Custom|NearlyDoneLifting) _returnvalue_7)
          (Ability|Tasks|ApplyRootMotionConstantForce "None" "0, 0,1.000000" (Variables|Default|GetBeamLiftStrength) _returnvalue_7 false "/AwfulDuck/Items/AwfulDuck/Curve_Athena_AwfulDuck_Curve.Curve_Athena_AwfulDuck_Curve" "SetVelocity") ; Beam Lift
          (Camera|PushCameraMode "/Game/Blueprints/Camera/Athena/Athena_PlayerCameraMode_HeldObject.Athena_PlayerCameraMode_HeldObject_C") ; Pull back camera
          (Class|FortPlayerParachute|SetChuteMeshVisibility (Variables|/Script/FortniteGame.FortPlayerPawn|GetParachuteAttachment (Variables|Default|GetFortPawn)) false) ; Hide glider
        )))))

(event Custom|QueryFinished
  (params (QueryInstance :type EQS%20Query%20Instance%20Object%20Reference) (QueryStatus :type EEnvQueryStatus%20Enum :default "Processing"))
  ;[1 1 1] Teleport to Ground
  ;;[1 1 1] Environment Query
  (switch Utilities|FlowControl|Switch|SwitchonEEnvQueryStatus QueryStatus
    (:Success
      (Utilities|IsValid QueryInstance
        (:Is%20Valid
          (bind _resultlocations (AI|EQS|GetQueryResultsasLocations QueryInstance))
          (Development|PrintString "EQS Finished")
          (if (> (Utilities|Array|Length _resultlocations) 0)
            (Variables|Default|SetSingleAbductReturnLocation (+ (Utilities|Array|RandomArrayItem _resultlocations) "0, 0,128.000000"))
            ;[1 1 1] QA logging
            (Fortnite|FortShippingLog (Variables|Default|GetInstigatorBGA) (Utilities|String|Append "AwfulDuck teleporting from (" (Utilities|String|ToString(Vector) (Transformation|GetActorLocation (Variables|Default|GetFortPawn))) ") to (" (Utilities|String|ToString(Vector) (Variables|Default|GetSingleAbductReturnLocation)) ") with a storm center at (" (Utilities|String|ToString(Vector) (SafeZone|GetSafeZoneCenter (SafeZone|GetSafeZoneIndicator))) ") and radius " (Utilities|String|ToString(Float) (SafeZone|GetSafeZoneRadius (SafeZone|GetSafeZoneIndicator)))))
            ;;[1 1 1] Teleport
            (Development|PrintString "Begin teleport")
            (Class|FortAsyncActionTeleportandWaitforLevelStreaming|CreateTeleportandWaitforLevelStreaming (Variables|Default|GetFortPawn) (Variables|Default|GetSingleAbductReturnLocation) (Math|Rotator|FindLookatRotation (Variables|Default|GetSingleAbductReturnLocation) (Math|Vector|MakeVector 0.0))
              (:OnComplete
                (Utilities|Time|SetTimerbyEvent (delegate Custom|TimerFinished3) (ScalableFloat|GetValueatLevel (Variables|Default|GetDelayAfterTeleport))) ; Delay after teleport
              ))
            (else
              (Development|PrintString "No EQS Results")
              (Ability|EndAbility))))
        (:Is%20Not%20Valid
          (Development|PrintString "EQS Query Invalid"))) ; Query valid?
    )
    (:Failed
      (Development|PrintString "EQS Query Failed"))
    (:Aborted
      (Development|PrintString "EQS Query Aborted"))
    (:OwnerLost
      (Development|PrintString "EQS Query Owner Lost"))
    (:MissingParam
      (Development|PrintString "EQS Query MIssing Param")))
  (Ability|EndAbility))

(event Custom|TimerFinished3
  ;[1 1 1] Teleport to Ground
  ;;[1 1 1] Teleport
  (Ability|EndAbility))

(event Custom|TimerFInished1
  ;[1 1 1] Teleport to Motherhsip
  (Ability|SendGameplayEventtoActor (Variables|Default|GetFortPawn) "(TagName=\"Event.Athena.Item.AwfulDuck.Abduct.Complete\")" (Utilities|Struct|MakeGameplayEventData "(TagName=\"\")" 0 (Variables|Default|GetFortPawn)))
  (Quest|SendComplexCustomStatEvent (Quests|GetQuestManager (Fortnite|GetFortPlayerControllerfromActor (Variables|Default|GetFortPawn)) "Invalid") 0 "(GameplayTags=)" "(GameplayTags=((TagName=\"Athena.Quests.AwfulDuck.Abduct.Complete\")))" 1) ; Accolade: Abducted
  (Ability|EndAbility))

(event Ability|EventOnEndAbility
  (params (bWasCancelled :type Boolean))
  ;[1 1 1] EndAbility
  (Development|PrintText "INVTEXT(\"Start Cleanup\")")
  ;;[1 0.2708 0.2807] Rez in VFX Tag
  (Variables|Default|SetWas%20Cancelled bWasCancelled)
  (Projectile|RemoveGameplayCue%28Looping%29Local (Variables|Default|GetFortPawn) (Variables|Default|GetRezOutAndIn_Tag) (Ability|GameplayCue|MakeGameplayCueParameters 0.0))
  (Development|PrintString "Player Rez in" true true "(R=0.302975,G=1.000000,B=0.000000,A=1.000000)" 1e+01)
  (Utilities|IsServer
    (:true
      (if (not (Variables|Default|GetWas%20Cancelled))
        (Ability|ApplyGameplayEffectToOwner "/AwfulDuck/Items/AwfulDuck/GE_AwfulDuck_TagAbducted.GE_AwfulDuck_TagAbducted_C") ; Tag this pawn as already abducted
      )
      (Ability|RemoveGameplayCueFromOwner (Variables|Default|GetGC_Beam_Tag))
      (GameplayEffects|RemoveActiveGameplayEffectbySourceEffect (Ability|GetFortAbilitySystemComponent (Variables|Default|GetFortPawn)) "/AwfulDuck/Items/AwfulDuck/GE_Athena_AwfulDuck_WarmUpPulse.GE_Athena_AwfulDuck_WarmUpPulse_C" 0 -1))
    (:false
      _))
  (Development|PrintText "INVTEXT(\"Remove camera effect\")")
  (Projectile|RemoveGameplayCue%28Looping%29Local (Ability|GetAvatarActorfromActorInfo) (Variables|Default|GetTeleportTunnelVFX) (Ability|GameplayCue|MakeGameplayCueParameters 0.0) false)
  (bind _asfort_player_controller_gameplay (Utilities|Casting|CastToFortPlayerControllerGameplay (Fortnite|GetFortPlayerControllerfromActor (Variables|Default|GetFortPawn)))
    (:then
      (HUD|SetHUDElementVisibilitywithContext _asfort_player_controller_gameplay (Variables|Default|GetHUDElementTags) false "AwfulDuck" ""))
    (:CastFailed
      _))
  (Development|PrintText "INVTEXT(\"Make pawn visible\")")
  (FortPawn|SetPawnVisibilityLocalControllerOnly (Variables|Default|GetFortPawn) false true)
  (Development|PrintText "INVTEXT(\"Restore camera\")")
  (Camera|PopCameraMode)
  ;;[1 1 1] Check falling, add immunity if needed
  (if (AI|Components|NavMovement|IsFalling (Variables|/Script/Engine.Character|GetCharacterMovement (Variables|Default|GetFortPawn)))
    (Development|PrintText "INVTEXT(\"Add fall immunity\")")
    (Fortnite|ApplyGameplayEffectsToActor (Variables|Default|GetFortPawn) (Utilities|Array|MakeArray "/Game/Athena/Items/Parents/FallImmunity/GE_Athena_TillLandFallDamageImmunity_Grant_Parent.GE_Athena_TillLandFallDamageImmunity_Grant_Parent_C") (Transformation|GetActorLocation (Variables|Default|GetFortPawn)))
    (if (Variables|Default|GetSingle%20Abduction)
      (Utilities|FlowControl|Delay 0.5)
      (Development|PrintString "Falling, wait a half second"))
    ;;[1 1 1] Stand up animation (single abduction only)
    (Development|PrintString "EndAbility")
    (GameplayEffects|RemoveActiveGameplayEffectbySourceEffect (Ability|GetFortAbilitySystemComponent (Variables|Default|GetFortPawn)) "/AwfulDuck/Items/AwfulDuck/GE_AwfulDuck_Lift.GE_AwfulDuck_Lift_C" 0 -1)
    (GameplayEffects|RemoveActiveGameplayEffectbySourceEffect (Ability|GetFortAbilitySystemComponent (Variables|Default|GetFortPawn)) "/AwfulDuck/Items/AwfulDuck/GE_AwfulDuck_LockControls.GE_AwfulDuck_LockControls_C" 0 -1)
    (|CallParentFunction (Variables|Default|GetWas%20Cancelled)))
  (Development|PrintText "INVTEXT(\"Not falling\")")
  (if (Variables|Default|GetSingle%20Abduction)
    (Utilities|IsServer
      (:true
        (if (not (Variables|Default|GetWas%20Cancelled))
          (bind _returnvalue (Ability|GameplayEffect|AssignTagSetbyCallerMagnitude (Ability|MakeOutgoingGameplayEffectSpec (Variables|Default|GetPostAbductionHealingEffect) 0.0) "(TagName=\"SetByCaller.Healing\")" (ScalableFloat|GetValueatLevel (Variables|Default|GetTeleportHealthRestore))))
          (bind _returnvalue_2 (Ability|GameplayEffect|AssignTagSetbyCallerMagnitude _returnvalue "(TagName=\"SetByCaller.Shield\")" (ScalableFloat|GetValueatLevel (Variables|Default|GetTeleportShieldRestore))))
          (Ability|ApplyGameplayEffectSpecToTarget _returnvalue_2 (Ability|TargetData|AbilityTargetDatafromActor (Variables|Default|GetFortPawn)))
          (if (not (AI|Components|NavMovement|IsFalling (Variables|/Script/Engine.Character|GetCharacterMovement (Variables|Default|GetFortPawn))))
            (Development|PrintString "Apply get up animation")
            (Ability|ApplyGameplayEffectToOwner "/AwfulDuck/Items/AwfulDuck/GE_AwfulDuck_LockControls.GE_AwfulDuck_LockControls_C"))))
      (:false
        _))
    (else
      _)))

(event Custom|NearlyDoneLifting
  (bind (_days _hours _minutes _seconds _milliseconds) (Math|Timespan|BreakTimespan (Math|DateTime|DateTime-DateTime (Math|DateTime|UTCNow) (Variables|Default|GetNewVar_0))))
  ;[1 1 1] Abduction effects
  ;;[1 1 1] Lifting beam
  (Ability|RemoveGameplayCueFromOwner (Variables|Default|GetGC_Beam_Tag))
  (Utilities|IsServer
    (:true
      (GameplayEffects|RemoveActiveGameplayEffectbySourceEffect (Ability|GetFortAbilitySystemComponent (Variables|Default|GetFortPawn)) "/AwfulDuck/Items/AwfulDuck/GE_Athena_AwfulDuck_WarmUpPulse.GE_Athena_AwfulDuck_WarmUpPulse_C" 0 -1)
      ;[1 1 1] Where we going?
      (if (Variables|Default|GetSingle%20Abduction)
        ;[1 1 1] Teleport to Ground
        (Utilities|IsServer
          (:true
            ;;[1 1 1] Environment Query
            (Development|PrintString "Begin EQS")
            (bind _returnvalue (AI|EQS|RunEQSQuery "/AwfulDuck/Items/AwfulDuck/EQS_AwfulDuckReturnPlacement.EQS_AwfulDuckReturnPlacement" (Variables|Default|GetFortPawn) "RandomBest25Pct"))
            (AI|EQS|BindEventtoOnQueryFinished _returnvalue (delegate Custom|QueryFinished))
            (Variables|Default|SetTeleportStartTime (Utilities|Time|GetRealTimeSeconds))))
        (else
          ;[1 1 1] Teleport to Motherhsip
          (Utilities|IsServer
            (:true
              (bind _returnvalue_2 (CallFunction|CheckShouldAbort))
              (if _returnvalue_2
                (Ability|CancelAbility)
                (else
                  (Development|PrintString (Utilities|String|Append "Abort check:" (Utilities|String|ToString(Float) (+ (Math|Conversions|ToFloat(Integer) _seconds) (/ (Math|Conversions|ToFloat(Integer) _milliseconds) 1e+03)))))
                  (Utilities|IsValid (Variables|Default|GetAbductionManager)
                    (:Is%20Valid
                      (bind _transform (Class|BAwfulDuckManager|GetSkyfireDestination (Variables|Default|GetAbductionManager) (Variables|Default|GetFortPawn)))
                      (Ability|SendGameplayEventtoActor (Variables|Default|GetFortPawn) "(TagName=\"Event.Athena.Item.AwfulDuck.Abduct.Teleport\")")
                      (Class|FortAsyncActionTeleportandWaitforLevelStreaming|CreateTeleportandWaitforLevelStreaming (Variables|Default|GetFortPawn) (.location _transform) (.rotation _transform)
                        (:OnComplete
                          _)))
                    (:Is%20Not%20Valid
                      (Development|PrintText "INVTEXT(\"No control tower\")")))
                  (Utilities|Time|SetTimerbyEvent (delegate Custom|TimerFInished1) (ScalableFloat|GetValueatLevel (Variables|Default|GetDelayAfterTeleport))) ; Delay after teleport
                )))))))
    (:false
      (Projectile|AddGameplayCue%28Looping%29Local (Variables|Default|GetFortPawn) (Variables|Default|GetTeleportTunnelVFX) (Ability|GameplayCue|MakeGameplayCueParameters 0.0)) ; Add local warp tunnel effect
      (bind _asfort_player_controller_gameplay (Utilities|Casting|CastToFortPlayerControllerGameplay (Fortnite|GetFortPlayerControllerfromActor (Variables|Default|GetFortPawn))))
      (HUD|SetHUDElementVisibilitywithContext _asfort_player_controller_gameplay (Variables|Default|GetHUDElementTags) true "AwfulDuck" ""))))

(fn CheckShouldAbort
  (bind _fortpawn (Variables|Default|GetFortPawn))
  (bind _returnvalue (Class|FortPlayerPawn|GetDBNOHoister _fortpawn))
  (Utilities|FlowControl|Sequence
    (:then_0
      ;[1 1 1] Hiding Prop
      (if (Class|FortPawn|IsInsideHidingProp (Variables|Default|GetFortPawn) true)
        (return true)))
    (:then_1
      ;[1 1 1] Carrying teammate or being carried
      (bind _bdidactorhavetags (Class|AthenaQuestFunctionLibrary|AthenaCheckActorHasTags (Variables|Default|GetFortPawn) "(GameplayTags=((TagName=\"Gameplay.Action.Player.CarryingPlayer\")))" true))
      (if _bdidactorhavetags
        (return true)))
    (:then_2
      (Utilities|IsValid _returnvalue
        (:Is%20Valid
          (bind _bdidactorhavetags_2 (Class|AthenaQuestFunctionLibrary|AthenaCheckActorHasTags _returnvalue "(GameplayTags=((TagName=\"Gameplay.Action.Player.CarryingPlayer\")))" true))
          (if _bdidactorhavetags_2
            (return true)))))
    (:then_3
      ;[1 1 1] In vehicle
      (if (Vehicle|IsinVehicle (Variables|Default|GetFortPawn))
        (return true)))
    (:then_4
      ;[1 1 1] Holding object
      (if (Class|FortPawn|GetIsHoldingObject (Variables|Default|GetFortPawn))
        (return true)))
    (:then_5
      ;[1 1 1] Zipline
      (bind _returnvalue_2 (Zipline|IsZiplining (Variables|Default|GetFortPawn)))
      (if _returnvalue_2
        (return true)))
    (:then_6
      ;[1 1 1] Exit bush
      (Ability|SendGameplayEventtoActor (Variables|Default|GetFortPawn) "(TagName=\"Abilities.Generic.Athena.Bush.Destroyed\")"))))

(fn IsPlayerDead
  (return (or (FortPawn|IsDead (Variables|Default|GetFortPawn)) (<= (FortPawn|GetHealth (Variables|Default|GetFortPawn)) 0.0))))

(fn IsSingleAbduction
  (bind _instigatorbga (Variables|Default|GetInstigatorBGA))
  (Utilities|IsValid _instigatorbga
    (:Is%20Valid
      (bind _asbga_awful_duck (Utilities|Casting|CastToBGA_AwfulDuck _instigatorbga)
        (:then
          (return (Variables|/AwfulDuck/Items/AwfulDuck/BGA_AwfulDuck.BGA_AwfulDuck_C|GetSingleAbduction _asbga_awful_duck)))
        (:CastFailed
          _)))
    (:Is%20Not%20Valid
      _)) ; Do we have a BGA?
  (Development|PrintText "INVTEXT(\"No BGA\")"))
