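; Gameplay ability for the Awful Duck item that shows warning gameplay cues on the player pawn it is active on.
; On activation it adds a replicated warning cue and a looping local FX cue, skipping either if the pawn already has that tag.
; Cue parameters carry the granting effect context, with its instigator as source object and, for the replicated cue, location.
; On end it removes both cues, calls the parent and prints a debug string.
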(event Ability|EventActivateAbility
  (bind _returnvalue (Ability|GetGrantedbyEffectContext))
  (bind _returnvalue_2 (Ability|EffectContext|GetInstigatorActor _returnvalue))
  (bind _returnvalue_3 (Ability|GetAvatarActorfromActorInfo))
  (bind _returnvalue_4 (Ability|GetGrantedbyEffectContext))
  (bind _asfort_player_pawn (Utilities|Casting|CastToFortPlayerPawn _returnvalue_3))
  ;[0.03905 0.06188 0.0625] New Added, need to pass parameters to the GC
  (Utilities|FlowControl|Sequence
    (:then_0
      (if (not (GameplayTags|HasMatchingGameplayTag _asfort_player_pawn (Variables|Default|GetTag_GC_Warning)))
        (Ability|AddGameplayCueWithParamsToOwner (Variables|Default|GetTag_GC_Warning) (Ability|GameplayCue|MakeGameplayCueParameters :NormalizedMagnitude 1.0 :RawMagnitude 1.0 :EffectContext _returnvalue_4 :Location (Transformation|GetActorLocation (Ability|EffectContext|GetInstigatorActor (Ability|GetGrantedbyEffectContext))) :Normal "0,0,0" :SourceObject _returnvalue_2 :GameplayEffectLevel 1 :AbilityLevel 1 :bReplicateLocationWhenUsingMinimalRepProxy true))))
    (:then_1
      (if (not (GameplayTags|HasMatchingGameplayTag _asfort_player_pawn (Variables|Default|GetTag_GC_Warning_LocalFX)))
        (Projectile|AddGameplayCue%28Looping%29Local _asfort_player_pawn (Variables|Default|GetTag_GC_Warning_LocalFX) (Ability|GameplayCue|MakeGameplayCueParameters :NormalizedMagnitude 0.0 :RawMagnitude 0.0 :EffectContext _returnvalue_4 :Location "0,0,0" :Normal "0,0,0" :SourceObject _returnvalue_2 :GameplayEffectLevel 1 :AbilityLevel 1 :bReplicateLocationWhenUsingMinimalRepProxy true))))))

(event Ability|EventOnEndAbility
  (params (bWasCancelled :type Boolean))
  ;[0.03905 0.06188 0.0625] New Added, need to pass parameters to the GC
  (Projectile|RemoveGameplayCue%28Looping%29Local (Utilities|Casting|CastToFortPlayerPawn (Ability|GetAvatarActorfromActorInfo)) (Variables|Default|GetTag_GC_Warning_LocalFX) (Ability|GameplayCue|MakeGameplayCueParameters :NormalizedMagnitude 0.0 :RawMagnitude 0.0 :EffectContext (Ability|GetGrantedbyEffectContext) :Location "0,0,0" :Normal "0,0,0" :SourceObject (Ability|EffectContext|GetInstigatorActor (Ability|GetGrantedbyEffectContext)) :GameplayEffectLevel 1 :AbilityLevel 1 :bReplicateLocationWhenUsingMinimalRepProxy true))
  (Ability|RemoveGameplayCueFromOwner (Variables|Default|GetTag_GC_Warning))
  (|CallParentFunction bWasCancelled)
  (Development|PrintString "Warning ability ended"))
