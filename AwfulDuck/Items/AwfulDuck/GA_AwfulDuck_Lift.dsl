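; GA_AwfulDuck_Lift is a gameplay ability Blueprint for the AwfulDuck item's lift action.
; On activation and on end it prints a debug string, then calls the parent implementation.
; The lift behaviour itself comes from the parent class; this graph adds only the logging.
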(event Ability|EventActivateAbility
  (Development|PrintString "Activate lift")
  (|CallParentFunction))

(event Ability|EventOnEndAbility
  (params (bWasCancelled :type Boolean))
  (Development|PrintString "End lift")
  (|CallParentFunction bWasCancelled))
