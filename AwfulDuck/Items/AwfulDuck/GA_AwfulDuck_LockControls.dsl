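; Gameplay ability for the AwfulDuck item that plays a lying-to-stand montage and lifts the control lock when it ends.
; Activation records the start time, starts the montage, calls the parent, and ends the ability when the montage completes.
; On end it removes the LockControls gameplay effect from the avatar pawn and prints the elapsed time.
; The lock effect is not applied in these graphs, and montage interruption has no handler.
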(event Ability|EventActivateAbility
  (Development|PrintString "Lock movement")
  (Variables|Default|SetTimeStarted (Utilities|Time|GetTimeSeconds))
  (Development|PrintString "Start montage")
  (Ability|Tasks|PlayMontageAndWait "None" "/AwfulDuck/Items/AwfulDuck/Anims/LyingToStandAwfulDuck_Montage.LyingToStandAwfulDuck_Montage" 1.5 "" false
    (:then
      (|CallParentFunction))
    (:OnCompleted
      (Development|PrintString "Montage finished")
      (Ability|EndAbility))))

(event Ability|EventOnEndAbility
  (params (bWasCancelled :type Boolean))
  (bind _asfort_pawn (Utilities|Casting|CastToFortPawn (Ability|GetAvatarActorfromActorInfo)))
  (GameplayEffects|RemoveActiveGameplayEffectbySourceEffect (Ability|GetFortAbilitySystemComponent _asfort_pawn) "/AwfulDuck/Items/AwfulDuck/GE_AwfulDuck_LockControls.GE_AwfulDuck_LockControls_C" 0 -1)
  (Development|PrintString (Utilities|String|Append "Movement unlocked ( " (Utilities|String|ToString(Float) (- (Utilities|Time|GetTimeSeconds) (Variables|Default|GetTimeStarted))) " seconds elapsed)"))
  (|CallParentFunction bWasCancelled))

(event Ability|EventFailedToActivatePassiveAbility
  (params (ActorInfo :type Gameplay%20Ability%20Actor%20Info%20Structure))
  (Development|PrintString "Ability failed"))
